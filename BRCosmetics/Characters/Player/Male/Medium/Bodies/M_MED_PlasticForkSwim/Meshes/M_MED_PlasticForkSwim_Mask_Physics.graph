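# Physics asset for the PlasticForkSwim male medium body's skull mask, covering its left and right back dyn_skullmask bone chains.
# Each side links a two-capsule 01 body to a tiny MSTR anchor capsule through one constraint with 90-degree swing and twist limit values.
# Large capsules on both thighs have no constraints, apparently serving as collision volumes for the mask pieces.
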PhysicsAsset /BRCosmetics/Characters/Player/Male/Medium/Bodies/M_MED_PlasticForkSwim/Meshes/M_MED_PlasticForkSwim_Mask_Physics
bodies {
 dyn_skullmask_baack_MSTR_r [{shapeType=Capsule, length=0.505, radius=0.505, shapeName=dyn_skullmask_baack_MSTR_r_capsule, rotation=<rot 0 -90 0>}]
 dyn_skullmask_back_01_l [
  {
   shapeType=Capsule
   length=5.6693
   radius=0.665
   shapeName=dyn_skullmask_back_01_l_capsule
   center=<2.1681 -1.7884 15.2799>
   rotation=<rot 6.9509 -123.1279 -61.874>
  }
  {shapeType=Capsule, length=7.1442, radius=0.9772, shapeName=dyn_skullmask_back_01_l_capsule, center=<4.6012 -4.1064 7.587>}
 ]
 dyn_skullmask_back_01_r [
  {
   shapeType=Capsule
   length=2.9777
   radius=1.005
   shapeName=dyn_skullmask_back_01_r_capsule
   center=<1.1457 1.7671 13.8857>
   rotation=<rot -50.6622 -82.9763 -37.1153>
  }
  {shapeType=Capsule, length=4.7473, radius=1.0826, shapeName=dyn_skullmask_back_01_r_capsule, center=<7.6429 4.8686 7.7644>}
 ]
 dyn_skullmask_back_MSTR_l [{shapeType=Capsule, length=0.505, radius=0.505, shapeName=dyn_skullmask_back_MSTR_l_capsule, rotation=<rot 0 -90 0>}]
 thigh_l [{shapeType=Capsule, length=32.2276, radius=7.2654, shapeName=thigh_r_capsule, center=<-23.33 0 1.024>, rotation=<rot 90 0 180>}]
 thigh_r [{shapeType=Capsule, length=32.2276, radius=7.2654, shapeName=thigh_r_capsule, center=<23.33 0 -1.024>, rotation=<rot 90 0 0>}]
}
constraints [
 {bone1Name=dyn_skullmask_back_01_r, bone2Name=dyn_skullmask_baack_MSTR_r, swing1LimitDegrees=90, swing2LimitDegrees=90, twistLimitDegrees=90}
 {bone1Name=dyn_skullmask_back_01_l, bone2Name=dyn_skullmask_back_MSTR_l, swing1LimitDegrees=90, swing2LimitDegrees=90, twistLimitDegrees=90}
]
