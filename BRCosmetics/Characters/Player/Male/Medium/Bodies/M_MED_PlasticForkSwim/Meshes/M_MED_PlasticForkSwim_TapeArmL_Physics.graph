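# Physics asset for the left-arm tape piece on the male medium PlasticForkSwim cosmetic body, set up for secondary motion of its dyn_tapearm bones.
# A five-link capsule chain hangs from a small master body constrained to the upper arm, with swing and twist limits per link and a very tight twist at the root.
# Larger upper arm, lower arm and thigh capsules approximate the body, with the thigh left unconstrained, likely as a collider.
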PhysicsAsset /BRCosmetics/Characters/Player/Male/Medium/Bodies/M_MED_PlasticForkSwim/Meshes/M_MED_PlasticForkSwim_TapeArmL_Physics
bodies {
 dyn_tapearm_01_l [{shapeType=Capsule, length=0.8778, radius=1.1454, shapeName=dyn_tapearm_01_l_capsule, center=<0.9402 0 0>, rotation=<rot 90 180 -179.9989>}]
 dyn_tapearm_02_l [{shapeType=Capsule, length=3.0667, radius=1.2433, shapeName=dyn_tapearm_02_l_capsule, center=<1.9647 0 0>, rotation=<rot 90 0 0>}]
 dyn_tapearm_03_l [{shapeType=Capsule, length=3.1925, radius=1.2792, shapeName=dyn_tapearm_03_l_capsule, center=<2.0537 0 0>, rotation=<rot 90 0 0>}]
 dyn_tapearm_04_l [{shapeType=Capsule, length=2.8356, radius=1.2209, shapeName=dyn_tapearm_04_l_capsule, center=<2.0466 0 0>, rotation=<rot 90 0 0>}]
 dyn_tapearm_05_l [{shapeType=Capsule, length=2.2245, radius=1.3939, shapeName=dyn_tapearm_05_l_capsule, center=<1.7087 0 0>, rotation=<rot 90 0 0>}]
 dyn_tapearm_MSTR_l [{shapeType=Capsule, length=0.505, radius=0.505, shapeName=dyn_tapearm_MSTR_l_capsule, rotation=<rot 0 -90 0>}]
 lowerarm_l [{shapeType=Capsule, length=19.1396, radius=3.4575, shapeName=lowerarm_l_capsule, center=<9.3292 0 0>, rotation=<rot 90 6.7213 6.7213>}]
 thigh_l [{shapeType=Capsule, length=24.1255, radius=6.9335, shapeName=thigh_l_capsule, center=<-19.4263 0.7818 0.7244>, rotation=<rot 90 0 0>}]
 upperarm_l [{shapeType=Capsule, length=21.7701, radius=4.212, shapeName=upperarm_l_capsule, center=<13.0406 0 -1.2754>, rotation=<rot 90 180 180>}]
}
constraints [
 {
  bone1Name=lowerarm_l
  bone2Name=upperarm_l
  swing1LimitDegrees=45
  swing1Motion=Free
  swing2LimitDegrees=45
  swing2Motion=Free
  twistLimitDegrees=45
  twistMotion=Free
 }
 {
  bone1Name=dyn_tapearm_MSTR_l
  bone2Name=upperarm_l
  swing1LimitDegrees=45
  swing1Motion=Free
  swing2LimitDegrees=45
  swing2Motion=Free
  twistLimitDegrees=45
  twistMotion=Free
 }
 {bone1Name=dyn_tapearm_01_l, bone2Name=dyn_tapearm_MSTR_l, swing1LimitDegrees=45, swing2LimitDegrees=45, twistLimitDegrees=1}
 {bone1Name=dyn_tapearm_02_l, bone2Name=dyn_tapearm_01_l, swing1LimitDegrees=45, swing2LimitDegrees=45, twistLimitDegrees=30}
 {bone1Name=dyn_tapearm_03_l, bone2Name=dyn_tapearm_02_l, swing1LimitDegrees=45, swing2LimitDegrees=45, twistLimitDegrees=30}
 {bone1Name=dyn_tapearm_04_l, bone2Name=dyn_tapearm_03_l, swing1LimitDegrees=45, swing2LimitDegrees=45, twistLimitDegrees=30}
 {bone1Name=dyn_tapearm_05_l, bone2Name=dyn_tapearm_04_l, swing1LimitDegrees=45, swing2LimitDegrees=45, twistLimitDegrees=30}
]
