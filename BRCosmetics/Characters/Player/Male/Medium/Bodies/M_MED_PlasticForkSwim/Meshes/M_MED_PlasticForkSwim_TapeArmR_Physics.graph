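# Physics asset for the right-arm tape dangle on the male medium PlasticForkSwim cosmetic body.
# A chain of four capsules, widening toward the tip, hangs from a small master body linked to the right forearm, with swing and twist limits between segments and very tight twist at the first.
# Upper arm, forearm and an unconstrained right thigh capsule supply body shapes around the chain; the elbow and master joints rotate freely.
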PhysicsAsset /BRCosmetics/Characters/Player/Male/Medium/Bodies/M_MED_PlasticForkSwim/Meshes/M_MED_PlasticForkSwim_TapeArmR_Physics
bodies {
 dyn_tapearm_01_r [{
   shapeType=Capsule
   length=1.8944
   radius=0.9732
   shapeName=dyn_tapearm_01_r_capsule
   center=<1.1529 0.3944 0.6027>
   rotation=<rot 63.6717 -162.7395 168.6978>
  }]
 dyn_tapearm_02_r [{shapeType=Capsule, length=2.9346, radius=1.2612, shapeName=dyn_tapearm_02_r_capsule, center=<2.0214 0 0>, rotation=<rot 90 0 0>}]
 dyn_tapearm_03_r [{shapeType=Capsule, length=2.8425, radius=1.4626, shapeName=dyn_tapearm_03_r_capsule, center=<1.9838 0 0>, rotation=<rot 90 0 0>}]
 dyn_tapearm_04_r [{shapeType=Capsule, length=1.938, radius=1.756, shapeName=dyn_tapearm_04_r_capsule, center=<2.0752 0 0>, rotation=<rot 90 0 0>}]
 dyn_tapearm_MSTR_r [{shapeType=Capsule, length=0.505, radius=0.505, shapeName=dyn_tapearm_MSTR_r_capsule, rotation=<rot 90 0 0>}]
 lowerarm_r [{shapeType=Capsule, length=17.6039, radius=3.1756, shapeName=lowerarm_r_capsule, center=<-11.3756 0 0>, rotation=<rot 90 0 0>}]
 thigh_r [{shapeType=Capsule, length=23.8455, radius=7.5668, shapeName=thigh_r_capsule, center=<17.2663 -0.4487 -1.5504>, rotation=<rot 90 0 0>}]
 upperarm_r [{
   shapeType=Capsule
   length=18.7815
   radius=4.0126
   shapeName=upperarm_r_capsule
   center=<-13.3178 0.3567 1.0083>
   rotation=<rot 86.7247 -176.4014 -179.3978>
  }]
}
constraints [
 {
  bone1Name=lowerarm_r
  bone2Name=upperarm_r
  swing1LimitDegrees=45
  swing1Motion=Free
  swing2LimitDegrees=45
  swing2Motion=Free
  twistLimitDegrees=45
  twistMotion=Free
 }
 {bone1Name=dyn_tapearm_04_r, bone2Name=dyn_tapearm_03_r, swing1LimitDegrees=60, swing2LimitDegrees=60, twistLimitDegrees=45}
 {
  bone1Name=dyn_tapearm_MSTR_r
  bone2Name=lowerarm_r
  swing1LimitDegrees=45
  swing1Motion=Free
  swing2LimitDegrees=45
  swing2Motion=Free
  twistLimitDegrees=45
  twistMotion=Free
 }
 {bone1Name=dyn_tapearm_01_r, bone2Name=dyn_tapearm_MSTR_r, swing1LimitDegrees=55, swing2LimitDegrees=60, twistLimitDegrees=2}
 {bone1Name=dyn_tapearm_02_r, bone2Name=dyn_tapearm_01_r, swing1LimitDegrees=60, swing2LimitDegrees=60, twistLimitDegrees=45}
 {bone1Name=dyn_tapearm_03_r, bone2Name=dyn_tapearm_02_r, swing1LimitDegrees=60, swing2LimitDegrees=60, twistLimitDegrees=45}
]
