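# Physics asset for the TapeLeg mesh of the male medium PlasticForkSwim player body cosmetic, covering a small dynamic tape bone chain on the left thigh.
# Capsule bodies cover the tape bones and both thighs, with thigh_r left unconstrained, likely as a collider.
# The master tape bone joins thigh_l with free swing and twist.
# From it, a base bone forks into a two-bone left branch and three-bone right branch, each joint with tight twist limits.
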PhysicsAsset /BRCosmetics/Characters/Player/Male/Medium/Bodies/M_MED_PlasticForkSwim/Meshes/M_MED_PlasticForkSwim_TapeLeg_Physics
bodies {
 dyn_tapeleg_01 [{shapeType=Capsule, length=2.7488, radius=0.9637, shapeName=dyn_tapeleg_01_capsule, center=<3.9592 0 0>}]
 dyn_tapeleg_01_l [{shapeType=Capsule, length=0.4239, radius=0.9737, shapeName=dyn_tapeleg_01_l_capsule, center=<2.6907 -0.1478 0>, rotation=<rot 90 0 0>}]
 dyn_tapeleg_01_r [{shapeType=Capsule, length=0.5353, radius=1.2039, shapeName=dyn_tapeleg_01_r_capsule, center=<2.8089 0.2493 0>, rotation=<rot 90 0 0>}]
 dyn_tapeleg_02_l [{shapeType=Capsule, length=1.52, radius=1.0527, shapeName=dyn_tapeleg_02_l_capsule, center=<1.2635 -0.2001 0>, rotation=<rot 90 0 0>}]
 dyn_tapeleg_02_r [{shapeType=Capsule, length=2.3816, radius=1.1699, shapeName=dyn_tapeleg_02_r_capsule, center=<1.7608 0.2493 0>, rotation=<rot 90 0 0>}]
 dyn_tapeleg_03_r [{shapeType=Capsule, length=2.0961, radius=1.4534, shapeName=dyn_tapeleg_03_r_capsule, center=<1.5235 0.2493 0>, rotation=<rot 90 0 0>}]
 dyn_tapeleg_MSTR [{shapeType=Capsule, length=0.505, radius=0.505, shapeName=dyn_tapeleg_MSTR_capsule, rotation=<rot 90 0 0>}]
 thigh_l [{shapeType=Capsule, length=28.3188, radius=6.0968, shapeName=thigh_l_capsule, center=<-17.6632 0 0>, rotation=<rot 90 0 0>}]
 thigh_r [{
   shapeType=Capsule
   length=24.056
   radius=7.6609
   shapeName=thigh_r_capsule
   center=<19.1521 0 0.8482>
   rotation=<rot 85.7193 179.9997 179.9998>
  }]
}
constraints [
 {
  bone1Name=dyn_tapeleg_MSTR
  bone2Name=thigh_l
  swing1LimitDegrees=45
  swing1Motion=Free
  swing2LimitDegrees=45
  swing2Motion=Free
  twistLimitDegrees=45
  twistMotion=Free
 }
 {bone1Name=dyn_tapeleg_01, bone2Name=dyn_tapeleg_MSTR, swing1LimitDegrees=7, swing2LimitDegrees=45, twistLimitDegrees=1}
 {bone1Name=dyn_tapeleg_01_l, bone2Name=dyn_tapeleg_01, swing1LimitDegrees=25, swing2LimitDegrees=35, twistLimitDegrees=4}
 {bone1Name=dyn_tapeleg_02_l, bone2Name=dyn_tapeleg_01_l, swing1LimitDegrees=30, swing2LimitDegrees=45, twistLimitDegrees=15}
 {bone1Name=dyn_tapeleg_01_r, bone2Name=dyn_tapeleg_01, swing1LimitDegrees=20, swing2LimitDegrees=25, twistLimitDegrees=4}
 {bone1Name=dyn_tapeleg_02_r, bone2Name=dyn_tapeleg_01_r, swing1LimitDegrees=20, swing2LimitDegrees=35, twistLimitDegrees=15}
 {bone1Name=dyn_tapeleg_03_r, bone2Name=dyn_tapeleg_02_r, swing1LimitDegrees=30, swing2LimitDegrees=35, twistLimitDegrees=15}
]
