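# Animation Blueprint for the PlasticForkSwim male medium body cosmetic, adding secondary motion to its tape and skull mask bones.
# It copies the attached parent's pose, runs a deform Control Rig, then weighted constraints position the dynamic master bones.
# Four RigidBody nodes simulate tape and mask physics, followed by AnimDynamics and Trail controllers for tape, weighted in when bRigidBodyEnabled is false.
# The event graph sets that fallback alpha and updates GravityOverride through a deprecated function.
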AnimBlueprint /BRCosmetics/Characters/Player/Male/Medium/Bodies/M_MED_PlasticForkSwim/Meshes/M_MED_PlasticForkSwim_Skeleton_AnimBlueprint
target_skeleton=&/M_MED_PlasticForkSwim_Skeleton
graphs [
 AnimGraph {nodes [
   n1 = Get rigid_body_disabled() -> (rigid_body_disabled:double)
   n2 = Control Rig(Source=Copy Pose From Mesh() {bUseAttachedParent=true}) {
    alphaBoolBlend {blendOption=Linear}
    controlRigAssetReference {blueprintRigClass=&/M_MED_PlasticForkSwim_Deform_CtrlRig.M_MED_PlasticForkSwim_Deform_CtrlRig_C}
   }
   n3 = "Constraint - dyn_tapearm_MSTR_l (PPP)"!(
    ComponentPose=Local To Component(LocalPose=@n2)
    Alpha=1
    ConstraintWeights_0=0.438
    ConstraintWeights_1:float=0.217
    ConstraintWeights_2:float=0.345
   ) -> (Pose) {
    alphaBoolBlend {blendOption=Linear}
    boneToModify {boneName=dyn_tapearm_MSTR_l}
    constraintSetup [{targetBone {boneName=upperarm_l}}, {targetBone {boneName=upperarm_twist_01_l}}, {targetBone {boneName=upperarm_twist_02_l}}]
    constraintWeights [1,1,1]
   }
   n4 = "Constraint - dyn_tapearm_MSTR_r (PPPPP)"!(
    ComponentPose=@n3
    Alpha=1
    ConstraintWeights_0=0.418
    ConstraintWeights_1:float=0.235
    ConstraintWeights_2:float=0.091
    ConstraintWeights_3:float=0.226
    ConstraintWeights_4:float=0.03
   ) -> (Pose) {
    alphaBoolBlend {blendOption=Linear}
    boneToModify {boneName=dyn_tapearm_MSTR_r}
    constraintSetup [
     {targetBone {boneName=deform_elbow_out_r}}
     {targetBone {boneName=lowerarm_r}}
     {targetBone {boneName=lowerarm_twist_02_r}}
     {targetBone {boneName=upperarm_r}}
     {targetBone {boneName=upperarm_twist_02_r}}
    ]
    constraintWeights [1,1,1,1,1]
   }
   n5 = "Constraint - dyn_tapeleg_MSTR (PPPPPP)"!(
    ComponentPose=@n4
    Alpha=1
    ConstraintWeights_0=0.053
    ConstraintWeights_1:float=0.068
    ConstraintWeights_2:float=0.143
    ConstraintWeights_3:float=0.046
    ConstraintWeights_4:float=0.674
    ConstraintWeights_5:float=0.016
   ) -> (Pose) {
    alphaBoolBlend {blendOption=Linear}
    boneToModify {boneName=dyn_tapeleg_MSTR}
    constraintSetup [
     {targetBone {boneName=deform_glute_l}}
     {targetBone {boneName=deform_groin_l}}
     {targetBone {boneName=pelvis}}
     {targetBone {boneName=thigh_l}}
     {targetBone {boneName=thigh_twist_01_l}}
     {targetBone {boneName=thigh_twist_01_r}}
    ]
    constraintWeights [1,1,1,1,1,1]
   }
   n6 = "Constraint - dyn_skullmask_back_MSTR_l (PPP)"!(ComponentPose=@n5, Alpha=1, ConstraintWeights_0=0.134, ConstraintWeights_1:float=0.754, ConstraintWeights_2:float=0.112) -> (Pose) {
    alphaBoolBlend {blendOption=Linear}
    boneToModify {boneName=dyn_skullmask_back_MSTR_l}
    constraintSetup [{targetBone {boneName=deform_glute_l}}, {targetBone {boneName=pelvis}}, {targetBone {boneName=thigh_twist_01_l}}]
    constraintWeights [1,1,1]
   }
   n7 = "Constraint - dyn_skullmask_baack_MSTR_r (PPP)"!(ComponentPose=@n6, Alpha=1, ConstraintWeights_0=0.184, ConstraintWeights_1:float=0.549, ConstraintWeights_2:float=0.267) -> (Pose) {
    alphaBoolBlend {blendOption=Linear}
    boneToModify {boneName=dyn_skullmask_baack_MSTR_r}
    constraintSetup [{targetBone {boneName=deform_glute_r}}, {targetBone {boneName=pelvis}}, {targetBone {boneName=thigh_twist_01_r}}]
    constraintWeights [1,1,1]
   }
   n8 = RigidBody!(ComponentPose=@n7, Alpha, OverrideWorldGravity:Vector="0,0,-980", ExternalForce) -> (Pose) {
    bClampLinearTranslationLimitToRefPose=true
    bOverrideWorldGravity=true
    cachedBoundsScale=1.35
    overridePhysicsAsset=&/M_MED_PlasticForkSwim_TapeArmL_Physics
    simulationSpace=BaseBoneSpace
    alphaBoolBlend {blendOption=Linear}
    baseBoneRef {boneName=upperarm_l}
    componentAppliedLinearAccClamp=<3000 3000 3000>
    componentLinearAccScale=<0.5 0.4 0.4>
    componentLinearVelScale=<0.5 0.4 0.4>
    overrideWorldGravity {z=-980}
   }
   n9 = RigidBody(ComponentPose=@n8, OverrideWorldGravity="0,0,-980") {
    bClampLinearTranslationLimitToRefPose=true
    bOverrideWorldGravity=true
    cachedBoundsScale=1.35
    overridePhysicsAsset=&/M_MED_PlasticForkSwim_TapeArmR_Physics
    simulationSpace=BaseBoneSpace
    alphaBoolBlend {blendOption=Linear}
    baseBoneRef {boneName=lowerarm_r}
    componentAppliedLinearAccClamp=<3000 3000 3000>
    componentLinearAccScale=<0.5 0.4 0.4>
    componentLinearVelScale=<0.5 0.4 0.4>
    overrideWorldGravity {z=-980}
   }
   n10 = RigidBody(ComponentPose=@n9, OverrideWorldGravity="0,0,-980") {
    bClampLinearTranslationLimitToRefPose=true
    bOverrideWorldGravity=true
    cachedBoundsScale=1.35
    overridePhysicsAsset=&/M_MED_PlasticForkSwim_TapeLeg_Physics
    simulationSpace=BaseBoneSpace
    alphaBoolBlend {blendOption=Linear}
    baseBoneRef {boneName=thigh_l}
    componentAppliedLinearAccClamp=<2000 2000 2000>
    componentLinearAccScale=<0.4 0.3 0.3>
    componentLinearVelScale=<0.4 0.3 0.3>
    overrideWorldGravity {z=-980}
   }
   n11 = RigidBody(ComponentPose=@n10, OverrideWorldGravity="0,0,-980") {
    bClampLinearTranslationLimitToRefPose=true
    cachedBoundsScale=1.3
    overridePhysicsAsset=&/M_MED_PlasticForkSwim_Mask_Physics
    simulationSpace=BaseBoneSpace
    alphaBoolBlend {blendOption=Linear}
    baseBoneRef {boneName=pelvis}
    componentAppliedLinearAccClamp=<1 1 1>
    componentLinearAccScale=<0.1 0.1 0.1>
    componentLinearVelScale=<0.1 0.1 0.1>
    overrideWorldGravity {z=-980}
   }
   n12 = Anim Dynamics - Bone: dyn_tapearm_01_l(ComponentPose=@n11, Alpha=@n1, ExternalForce="0,0,0") {
    angularBiasOverride=0.95
    angularDampingOverride=0.95
    linearDampingOverride=0.95
    alphaBoolBlend {blendOption=Linear}
    boundBone {boneName=dyn_tapearm_01_l}
    componentAppliedLinearAccClamp=<1000 1000 1000>
    componentLinearAccScale=<0.3 0.3 0.3>
    componentLinearVelScale=<0.3 0.3 0.3>
    physicsBodyDefinitions [{
      boundBone {boneName=dyn_tapearm_01_l}
      boxExtents=<4 13 4>
      constraintSetup {angularLimitsMax=<2 15 0>, angularLimitsMin=<-2 -15 -30>, angularTarget=<0 0 0>}
      localJointOffset=<2 0 0>
     }]
    relativeSpaceBone {boneName=upperarm_twist_02_l}
   }
   n13 = Trail controller - Bone: dyn_tapearm_05_l(ComponentPose=@n12, Alpha=@n1) {
    bActorSpaceFakeVel=true
    bLimitRotation=true
    bUsePlanarLimit=true
    chainLength=4
    fakeVelocity=<1 1 1>
    rotationLimits [
     {limitMax=<10 25 0>, limitMin=<-10 -25 -45>}
     {limitMax=<10 25 5>, limitMin=<-10 -25 -45>}
     {limitMax=<10 25 5>, limitMin=<-10 -25 -45>}
     {limitMax=<10 25 5>, limitMin=<-10 -25 -45>}
    ]
    rotationOffsets [<0 0 0>, <0 0 0>, <0 0 0>, <0 0 0>]
    trailBone {boneName=dyn_tapearm_05_l}
   }
   n14 = Anim Dynamics - Bone: dyn_tapearm_01_r(ComponentPose=@n13, Alpha=@n1, ExternalForce="0,0,0") {
    angularBiasOverride=0.95
    angularDampingOverride=0.95
    linearDampingOverride=0.95
    alphaBoolBlend {blendOption=Linear}
    boundBone {boneName=dyn_tapearm_01_r}
    componentAppliedLinearAccClamp=<1000 1000 1000>
    componentLinearAccScale=<0.3 0.3 0.3>
    componentLinearVelScale=<0.3 0.3 0.3>
    physicsBodyDefinitions [{
      boundBone {boneName=dyn_tapearm_01_r}
      boxExtents=<4 13 4>
      constraintSetup {angularLimitsMax=<2 30 10>, angularLimitsMin=<-2 -20 -30>, angularTarget=<0 0 0>}
      localJointOffset=<0.5 0 0>
     }]
    relativeSpaceBone {boneName=lowerarm_r}
   }
   n15 = Trail controller - Bone: dyn_tapearm_04_r(ComponentPose=@n14, Alpha=@n1) {
    bActorSpaceFakeVel=true
    bLimitRotation=true
    bUsePlanarLimit=true
    chainLength=3
    fakeVelocity=<1 1 1>
    rotationLimits [{limitMax=<10 25 0>, limitMin=<-10 -25 -45>}, {limitMax=<10 25 5>, limitMin=<-10 -25 -45>}, {limitMax=<10 25 5>, limitMin=<-10 -25 -45>}]
    rotationOffsets [<0 0 0>, <0 0 0>, <0 0 0>]
    trailBone {boneName=dyn_tapearm_04_r}
   }
   n16 = Anim Dynamics - Bone: dyn_tapeleg_01(ComponentPose=@n15, Alpha=@n1, ExternalForce="0,0,0") {
    angularBiasOverride=0.95
    angularDampingOverride=0.95
    linearDampingOverride=0.95
    alphaBoolBlend {blendOption=Linear}
    boundBone {boneName=dyn_tapeleg_01}
    componentAppliedLinearAccClamp=<1000 1000 1000>
    componentLinearAccScale=<0.3 0.3 0.3>
    componentLinearVelScale=<0.3 0.3 0.3>
    physicsBodyDefinitions [{
      boundBone {boneName=dyn_tapeleg_01}
      boxExtents=<4 13 4>
      constraintSetup {angularLimitsMax=<2 15 0>, angularLimitsMin=<-2 -15 -15>, angularTarget=<0 0 0>}
      localJointOffset=<1 0 0>
     }]
    relativeSpaceBone {boneName=lowerarm_r}
   }
   n17 = Trail controller - Bone: dyn_tapeleg_02_l(ComponentPose=@n16, Alpha=@n1) {
    bActorSpaceFakeVel=true
    bLimitRotation=true
    bUsePlanarLimit=true
    relaxationSpeedScale=0.1
    fakeVelocity=<1 1 1>
    rotationLimits [{limitMax=<3 0 0>, limitMin=<-3 -25 -25>}, {limitMax=<3 0 5>, limitMin=<-3 -25 -25>}]
    rotationOffsets [<0 0 0>, <0 0 0>]
    trailBone {boneName=dyn_tapeleg_02_l}
   }
   n18 = Trail controller - Bone: dyn_tapeleg_03_r(ComponentPose=@n17, Alpha=@n1) {
    bActorSpaceFakeVel=true
    bLimitRotation=true
    bUsePlanarLimit=true
    relaxationSpeedScale=0.1
    fakeVelocity=<1 1 1>
    rotationLimits [{limitMax=<3 25 0>, limitMin=<-3 -25 -25>}, {limitMax=<3 25 0>, limitMin=<-3 -25 -25>}]
    rotationOffsets [<0 0 0>, <0 0 0>]
    trailBone {boneName=dyn_tapeleg_03_r}
   }
   Output Pose(Result=Component To Local(ComponentPose=@n18))
  ]}
 EventGraph {nodes [
   Event BlueprintUpdateAnimation()
   > n1 = Sequence()
   > n2 = Gravity Override State Function DEPRECATED(CharacterPartAnimInstance=Self-Reference(), GameTimeInSeconds=GetGameTimeInSeconds(), GravityOverrideParams=Get Gravity Override Params())
   > Set GravityOverride(GravityOverride=@n2)
   n3 = Branch(execute=@n1.then_1, Condition=Get bRigidBodyEnabled())
   > Set rigid_body_disabled(rigid_body_disabled:double=0) -> (Output_Get:double)
   Set rigid_body_disabled(execute=@n3.else, rigid_body_disabled=1)
   TryGetPawnOwner()
  ]}
]
variables [
 Gravity Override Params {type=GravityOverrideParamsStruct}
 GravityOverride {type=Vector}
 rigid_body_disabled {type=number}
]
