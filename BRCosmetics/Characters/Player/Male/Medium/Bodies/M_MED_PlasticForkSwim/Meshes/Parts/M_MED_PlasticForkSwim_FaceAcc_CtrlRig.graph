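# Control Rig for the PlasticForkSwim face accessory mesh that repositions hair base bones as the head and clavicles move.
# A sequence runs head-pitch offsets, aim constraints toward spine_05 with some head-weighted, layered pose correctives, and clavicle-driven offsets.
# Weights come from remapped relative Euler angles, distances to target poses and a spherical pose reader, with Kalman smoothing on the last corrective.
# Several corrective and debug print nodes are left unconnected.
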ControlRigBlueprint /BRCosmetics/Characters/Player/Male/Medium/Bodies/M_MED_PlasticForkSwim/Meshes/Parts/M_MED_PlasticForkSwim_FaceAcc_CtrlRig
graphs [
 Multiply_2_1_ContainedGraph {nodes [
   n1 = RigVMFunctionEntryNode()
   RigVMFunctionReturnNode(Result=&RigVMFunction_MathDoubleMul(B=@n1.C, A=&RigVMFunction_MathDoubleMul(A=@n1.A, B=@n1.B).Result).Result)
  ]}
 New Function_ContainedGraph {nodes [
   RigVMFunctionEntryNode()
   RigUnit_GetRelativeTransformForItem = &RigUnit_GetRelativeTransformForItem(Child=(Type=Bone,Name="head"), Parent=(Type=Bone,Name="spine_05"), bChildInitial=False, bParentInitial=False)
   RigVMFunction_MathQuaternionToEuler = &RigVMFunction_MathQuaternionToEuler(RotationOrder=ZYX, Value=@RigUnit_GetRelativeTransformForItem.RelativeTransform.Rotation)
   > RigVMFunctionReturnNode(Result=@RigVMFunction_MathQuaternionToEuler.Result)
  ]}
 RigVMFunctionLibrary {nodes [F_Head Rotation = RigVMCollapseNode()]}
 RigVMFunction_Sequence_1_ContainedGraph {nodes [
   n1 = RigVMFunctionEntryNode()
   RigVMFunction_Sequence = &RigVMFunction_Sequence(ExecuteContext=@n1.ExecuteContext)
   n2 = &RigVMFunction_Sequence(ExecuteContext=@RigVMFunction_Sequence.B)
   n3 = &RigVMFunction_Sequence(ExecuteContext=@n2.B)
   n4 = &RigVMFunction_Sequence(ExecuteContext=@n3.B)
   n5 = &RigVMFunction_Sequence(ExecuteContext=@n4.B)
   RigVMFunctionReturnNode(A=@RigVMFunction_Sequence.A, B=@n2.A, C=@n3.A, D=@n4.A, E=@n5.A, F=@n5.B)
  ]}
 RigVMModel {nodes [
   n1 = RigVMDispatchNode(Values="((Rotation=(X=0,Y=0,Z=0,W=1),Translation=(X=0,Y=0,Z=0),Scale3D=(X=1,Y=1,Z=1)),((Rotation=(X=0,Y=0,Z=0,W=1),Translation=(X=0,Y=0,Z=0),Scale3D=(X=1,Y=1,Z=1)),Rotation=(X=0,Y=0,Z=0,W=1),Translation=(X=-3,Y=0,Z=0),Scale3D=(X=1,Y=1,Z=1)),((Rotation=(X=0,Y=0,Z=0 ...(truncated)") {template="DISPATCH_RigVMDispatch_ArrayMake(in Values,out Array)"}
   F_Head Rotation_3 = RigVMFunctionReferenceNode()
   &RigUnit_ModifyTransforms(
    ExecutePin=()
    ItemToModify="((Item=(Type=Bone,Name=\"dyn_hairMID_01_Base_r\"),Transform=(Rotation=(X=0,Y=0,Z=0,W=1),Translation=(X=0,Y=0,Z=0),Scale3D=(X=1,Y=1,Z=1))),(Item=(Type=Bone,Name=\"dyn_hairBack_01_Base_r\"),Transform=(Rotation=(X=0,Y=0,Z=0,W=1),Translation=(X=-3,Y=0,Z=0),Scale3D ...(truncated)"
    WeightMaximum=0
    WeightMinimum=0
    ItemToModify.2.Transform=RigVMDispatchNode(Index=2, Array=@n1.Array) {template="DISPATCH_RigVMDispatch_ArrayGetAtIndex(in Array,in Index,out Element)"}.Element
    ItemToModify.1.Transform=RigVMDispatchNode(Index=1, Array=@n1.Array) {template="DISPATCH_RigVMDispatch_ArrayGetAtIndex(in Array,in Index,out Element)"}.Element
    ItemToModify.0.Transform=RigVMDispatchNode(Index=0, Array=@n1.Array) {template="DISPATCH_RigVMDispatch_ArrayGetAtIndex(in Array,in Index,out Element)"}.Element
    Weight=&RigVMFunction_MathDoubleMul(
     A=&RigVMFunction_MathDoubleRemap(SourceMaximum=-65, SourceMinimum=0, TargetMaximum=1, TargetMinimum=0, bClamp=False, Value=@"F_Head Rotation_3".Result.X).Result
     B=&RigVMFunction_MathDoubleRemap(SourceMaximum=-35, SourceMinimum=0, TargetMaximum=1, TargetMinimum=0, bClamp=False, Value=@"F_Head Rotation_3".Result.Z).Result
    ).Result
   )
   n2 = RigVMDispatchNode(Values="((Rotation=(X=0,Y=0,Z=-0.0656,W=0.9978),Translation=(X=-0.3828,Y=0,Z=0),Scale3D=(X=1,Y=1,Z=1)),((Rotation=(X=0,Y=0,Z=0,W=1),Translation=(X=0,Y=0,Z=0),Scale3D=(X=1,Y=1,Z=1)),Rotation=(X=0,Y=0,Z=-0.0204,W=0.9998),Translation=(X=-2.8475,Y=0.1162,Z=0),Scale3D= ...(truncated)") {template="DISPATCH_RigVMDispatch_ArrayMake(in Values,out Array)"}
   F_Head Rotation_3_1 = RigVMFunctionReferenceNode()
   > &RigUnit_ModifyTransforms(
    ItemToModify="((Item=(Type=Bone,Name=\"dyn_hairBack_01_Base_l\"),Transform=(Rotation=(X=0,Y=0,Z=-0.0656,W=0.9978),Translation=(X=-0.3828,Y=0,Z=0),Scale3D=(X=1,Y=1,Z=1))),(Item=(Type=Bone,Name=\"dyn_hairMid_01_Base_l\"),Transform=(Rotation=(X=0,Y=0,Z=-0.0204,W=0.9998),Transl ...(truncated)"
    WeightMaximum=0
    WeightMinimum=0
    ItemToModify.2.Transform=RigVMDispatchNode(Index=2, Array=@n2.Array) {template="DISPATCH_RigVMDispatch_ArrayGetAtIndex(in Array,in Index,out Element)"}.Element
    ItemToModify.1.Transform=RigVMDispatchNode(Index=1, Array=@n2.Array) {template="DISPATCH_RigVMDispatch_ArrayGetAtIndex(in Array,in Index,out Element)"}.Element
    ItemToModify.0.Transform=RigVMDispatchNode(Index=0, Array=@n2.Array) {template="DISPATCH_RigVMDispatch_ArrayGetAtIndex(in Array,in Index,out Element)"}.Element
    Weight=&RigVMFunction_MathDoubleMul(
     B=&RigVMFunction_MathDoubleRemap(SourceMaximum=-33, SourceMinimum=0, TargetMaximum=1, TargetMinimum=0, bClamp=False, Value=@"F_Head Rotation_3_1".Result.Z).Result
     A=&RigVMFunction_MathDoubleRemap(SourceMaximum=33, SourceMinimum=0, TargetMaximum=1, TargetMinimum=0, bClamp=False, Value=@"F_Head Rotation_3_1".Result.X).Result
    ).Result
   )
   RigUnit_BeginExecution = &RigUnit_BeginExecution()
   RigVMFunction_Sequence = RigVMAggregateNode(ExecuteContext=@RigUnit_BeginExecution.ExecutePin)
   n3 = RigVMDispatchNode(Values="((Rotation=(X=-0.0648,Y=0,Z=0,W=0.9979),Translation=(X=-1.8378,Y=0,Z=0),Scale3D=(X=0.6928,Y=1.114,Z=1.3015)),((Rotation=(X=0,Y=0,Z=0,W=1),Translation=(X=0,Y=0,Z=0),Scale3D=(X=1,Y=1,Z=1)),Rotation=(X=0,Y=0,Z=0,W=1),Translation=(X=-4.7825,Y=0,Z=0),Scale3D=(X ...(truncated)") {template="DISPATCH_RigVMDispatch_ArrayMake(in Values,out Array)"}
   RigUnit_GetRelativeTransformForItem = &RigUnit_GetRelativeTransformForItem(Child=(Type=Bone,Name="clavicle_l"), Parent=(Type=Bone,Name="spine_05"), bChildInitial=False, bParentInitial=False)
   RigVMFunction_MathQuaternionToEuler = &RigVMFunction_MathQuaternionToEuler(RotationOrder=YXZ, Value=@RigUnit_GetRelativeTransformForItem.RelativeTransform.Rotation)
   &RigUnit_ModifyTransforms(
    ItemToModify="((Item=(Type=Bone,Name=\"dyn_hairBack_01_Base_l\"),Transform=(Rotation=(X=-0.0648,Y=0,Z=0,W=0.9979),Translation=(X=-1.8378,Y=0,Z=0),Scale3D=(X=0.6928,Y=1.114,Z=1.3015))),(Item=(Type=Bone,Name=\"dyn_hairMid_01_Base_l\"),Transform=(Rotation=(X=0,Y=0,Z=0,W=1),Tra ...(truncated)"
    WeightMinimum=0
    ItemToModify.0.Transform=RigVMDispatchNode(Index=0, Array=@n3.Array) {template="DISPATCH_RigVMDispatch_ArrayGetAtIndex(in Array,in Index,out Element)"}.Element
    ItemToModify.1.Transform=RigVMDispatchNode(Index=1, Array=@n3.Array) {template="DISPATCH_RigVMDispatch_ArrayGetAtIndex(in Array,in Index,out Element)"}.Element
    ItemToModify.2.Transform=RigVMDispatchNode(Index=2, Array=@n3.Array) {template="DISPATCH_RigVMDispatch_ArrayGetAtIndex(in Array,in Index,out Element)"}.Element
    Weight=&RigVMFunction_MathDoubleRemap(SourceMaximum=30, SourceMinimum=80, TargetMaximum=1, TargetMinimum=0, bClamp=False, Value=@RigVMFunction_MathQuaternionToEuler.Result.Y).Result
    ExecutePin=@RigVMFunction_Sequence.E
   )
   n4 = RigVMDispatchNode(Values="((Rotation=(X=-0.0648,Y=0,Z=0,W=0.9979),Translation=(X=-1.8378,Y=0,Z=0),Scale3D=(X=0.6928,Y=1.114,Z=1.3015)),((Rotation=(X=0,Y=0,Z=0,W=1),Translation=(X=0,Y=0,Z=0),Scale3D=(X=1,Y=1,Z=1)),Rotation=(X=0,Y=0,Z=0,W=1),Translation=(X=-4.7825,Y=0,Z=0),Scale3D=(X ...(truncated)") {template="DISPATCH_RigVMDispatch_ArrayMake(in Values,out Array)"}
   n5 = &RigUnit_GetRelativeTransformForItem(Child=(Type=Bone,Name="clavicle_r"), Parent=(Type=Bone,Name="spine_05"), bChildInitial=False, bParentInitial=False)
   > &RigUnit_ModifyTransforms(
    ItemToModify="((Item=(Type=Bone,Name=\"dyn_hairBack_01_Base_r\"),Transform=(Rotation=(X=-0.0648,Y=0,Z=0,W=0.9979),Translation=(X=-1.8378,Y=0,Z=0),Scale3D=(X=0.6928,Y=1.114,Z=1.3015))),(Item=(Type=Bone,Name=\"dyn_hairMID_01_Base_r\"),Transform=(Rotation=(X=0,Y=0,Z=0,W=1),Tra ...(truncated)"
    WeightMinimum=0
    ItemToModify.2.Transform=RigVMDispatchNode(Index=2, Array=@n4.Array) {template="DISPATCH_RigVMDispatch_ArrayGetAtIndex(in Array,in Index,out Element)"}.Element
    ItemToModify.1.Transform=RigVMDispatchNode(Index=1, Array=@n4.Array) {template="DISPATCH_RigVMDispatch_ArrayGetAtIndex(in Array,in Index,out Element)"}.Element
    ItemToModify.0.Transform=RigVMDispatchNode(Index=0, Array=@n4.Array) {template="DISPATCH_RigVMDispatch_ArrayGetAtIndex(in Array,in Index,out Element)"}.Element
    Weight=&RigVMFunction_MathDoubleRemap(
     SourceMaximum=30
     SourceMinimum=80
     TargetMaximum=1
     TargetMinimum=0
     bClamp=False
     Value=&RigVMFunction_MathQuaternionToEuler(RotationOrder=YXZ, Value=@n5.RelativeTransform.Rotation).Result.Y
    ).Result
   )
   DISPATCH_RigVMDispatch_ArrayMake = RigVMDispatchNode(Values=((Rotation=(X=0,Y=0,Z=0,W=1),Translation=(X=-2.0932,Y=-6.8148,Z=0.0199),Scale3D=(X=1,Y=1,Z=1)))) {template="DISPATCH_RigVMDispatch_ArrayMake(in Values,out Array)"}
   DISPATCH_RigVMDispatch_ArrayGetAtIndex = RigVMDispatchNode(Index=0, Array=@DISPATCH_RigVMDispatch_ArrayMake.Array) {template="DISPATCH_RigVMDispatch_ArrayGetAtIndex(in Array,in Index,out Element)"}
   F_Head Rotation = RigVMFunctionReferenceNode()
   Remap = &RigVMFunction_MathDoubleRemap(SourceMaximum=-70, SourceMinimum=-30, TargetMaximum=1, TargetMinimum=0, bClamp=False, Value=@"F_Head Rotation".Result.Z)
   &RigUnit_ModifyTransforms(
    ItemToModify=((Item=(Type=Bone,Name="dyn_hairBack_01_c_BASE"),Transform=(Rotation=(X=0,Y=0,Z=0,W=1),Translation=(X=-2.0932,Y=-6.8148,Z=0.0199),Scale3D=(X=1,Y=1,Z=1))))
    WeightMinimum=0
    ItemToModify.0.Transform=@DISPATCH_RigVMDispatch_ArrayGetAtIndex.Element
    Weight=@Remap.Result
    ExecutePin=@RigVMFunction_Sequence.B
   )
   F_Head Rotation_2_2 = RigVMFunctionReferenceNode()
   Clamp = &RigVMFunction_MathDoubleClamp(
    Minimum=0.7
    Value=&RigVMFunction_MathDoubleSub(
     A=1
     B=&RigVMFunction_MathDoubleRemap(SourceMaximum=90, SourceMinimum=-10, TargetMaximum=1, TargetMinimum=0, bClamp=False, Value=@"F_Head Rotation_2_2".Result.Z).Result
    ).Result
   )
   &RigUnit_AimConstraintLocalSpaceOffset(
    Child=(Type=Bone,Name="dyn_hairBack_01_c_BASE")
    Filter=(bX=True,bY=True,bZ=True)
    Parents=((Item=(Type=Bone,Name="spine_05"),Weight=1))
    UpAxis=(X=0,Y=1,Z=0)
    WorldUp=(Target=(X=0,Y=50,Z=0),Kind=Location,Space=(Type=Bone,Name="head"))
    bMaintainOffset=true
    Weight=@Clamp.Result
    ExecutePin=@RigVMFunction_Sequence.C
   )
   > &RigUnit_AimConstraintLocalSpaceOffset(
    Child=(Type=Bone,Name="dyn_hairMID_01_Base_r")
    Filter=(bX=True,bY=True,bZ=True)
    Parents=((Item=(Type=Bone,Name="spine_05"),Weight=1))
    UpAxis=(X=0,Y=1,Z=0)
    Weight=1
    WorldUp=(Target=(X=0,Y=50,Z=0),Kind=Location,Space=(Type=Bone,Name="head"))
    bMaintainOffset=true
   )
   > &RigUnit_AimConstraintLocalSpaceOffset(
    Child=(Type=Bone,Name="dyn_hairMid_01_Base_l")
    Filter=(bX=True,bY=True,bZ=True)
    Parents=((Item=(Type=Bone,Name="spine_05"),Weight=1))
    UpAxis=(X=0,Y=1,Z=0)
    Weight=1
    WorldUp=(Target=(X=0,Y=50,Z=0),Kind=Location,Space=(Type=Bone,Name="head"))
    bMaintainOffset=true
   )
   F_Head Rotation_1 = RigVMFunctionReferenceNode()
   F_Head Rotation_2 = RigVMFunctionReferenceNode()
   n6 = &RigVMFunction_MathDoubleSub(
    A=1
    B=&RigVMFunction_MathDoubleRemap(SourceMaximum=250, SourceMinimum=-10, TargetMaximum=1, TargetMinimum=0, bClamp=False, Value=@"F_Head Rotation_2".Result.Z).Result
   )
   Multiply = &RigVMFunction_MathDoubleMul(
    A=&RigVMFunction_MathDoubleSub(
     A=1
     B=&RigVMFunction_MathDoubleRemap(SourceMaximum=-40, SourceMinimum=0, TargetMaximum=1, TargetMinimum=0, Value=@"F_Head Rotation_1".Result.Y).Result
    ).Result
    B=@n6.Result
   )
   > &RigUnit_AimConstraintLocalSpaceOffset(
    AdvancedSettings=(DebugSettings=(bEnabled=false,Scale=10,WorldOffset=(Rotation=(X=0,Y=0,Z=0,W=1),Translation=(X=0,Y=0,Z=0),Scale3D=(X=1,Y=1,Z=1))),RotationOrderForFilter=XZY)
    Child=(Type=Bone,Name="dyn_hairBack_01_Base_l")
    Filter=(bX=True,bY=True,bZ=True)
    Parents=((Item=(Type=Bone,Name="spine_05"),Weight=1))
    WorldUp=(Target=(X=10,Y=0,Z=-25),Kind=Location,Space=(Type=Bone,Name="head"))
    bMaintainOffset=true
    Weight=@Multiply.Result
   )
   F_Head Rotation_1_1 = RigVMFunctionReferenceNode()
   Subtract = &RigVMFunction_MathDoubleSub(
    A=1
    B=&RigVMFunction_MathDoubleRemap(SourceMaximum=40, SourceMinimum=0, TargetMaximum=1, TargetMinimum=0, Value=@"F_Head Rotation_1_1".Result.Y).Result
   )
   > &RigUnit_AimConstraintLocalSpaceOffset(
    AdvancedSettings=(DebugSettings=(bEnabled=false,Scale=10,WorldOffset=(Rotation=(X=0,Y=0,Z=0,W=1),Translation=(X=0,Y=0,Z=0),Scale3D=(X=1,Y=1,Z=1))),RotationOrderForFilter=XZY)
    Child=(Type=Bone,Name="dyn_hairBack_01_Base_r")
    Filter=(bX=True,bY=True,bZ=True)
    Parents=((Item=(Type=Bone,Name="spine_05"),Weight=1))
    WorldUp=(Target=(X=10,Y=0,Z=25),Kind=Location,Space=(Type=Bone,Name="head"))
    bMaintainOffset=true
    Weight=&RigVMFunction_MathDoubleMul(A=@Subtract.Result, B=@n6.Result).Result
   )
   F_Head Rotation_4 = RigVMFunctionReferenceNode()
   Power = /Script/RigVM.RigVMFunction_MathDoublePow(
    B=0.7
    A=&RigVMFunction_MathDoubleRemap(SourceMaximum=-55, SourceMinimum=0, TargetMaximum=1, TargetMinimum=0, bClamp=False, Value=@"F_Head Rotation_4".Result.Z).Result
   ) {is=struct RigVMUnitNode, template="Power::Execute(in A,in B,out Result)"}
   &RigUnit_ModifyTransforms(
    ItemToModify="((Item=(Type=Bone,Name=\"dyn_hairBack_01_c_BASE\"),Transform=(Rotation=(X=0,Y=0,Z=0,W=1),Translation=(X=-2,Y=-3,Z=0),Scale3D=(X=1,Y=1,Z=1))),(Item=(Type=Bone,Name=\"dyn_hairMID_01_Base_r\"),Transform=(Rotation=(X=0,Y=0,Z=0,W=1),Translation=(X=-3,Y=0,Z=0),Scale ...(truncated)"
    WeightMinimum=0
    Weight=@Power.Result
    ExecutePin=@RigVMFunction_Sequence.D
   )
   n7 = &RigVMFunction_MathQuaternionToEuler(
    RotationOrder=ZYX
    Value=&RigUnit_GetRelativeTransformForItem(Child=(Type=Bone,Name="head"), Parent=(Type=Null,Name="head_null"), bChildInitial=False, bParentInitial=False).RelativeTransform.Rotation
   )
   n8 = &RigUnit_GetRelativeTransformForItem(Child=(Type=Bone,Name="clavicle_r"), Parent=(Type=Null,Name="clavicle_r_null"), bChildInitial=False, bParentInitial=False)
   Multiply_2 = RigVMAggregateNode(
    A=&RigVMFunction_MathDoubleRemap(SourceMaximum=-20, SourceMinimum=-5, TargetMaximum=1, TargetMinimum=0, Value=@n7.Result.Y).Result
    B=&RigVMFunction_MathDoubleRemap(SourceMaximum=-20, SourceMinimum=5, TargetMaximum=1, TargetMinimum=0, Value=@n7.Result.Z).Result
    C=&RigVMFunction_MathDoubleRemap(
     SourceMaximum=-14
     SourceMinimum=0
     TargetMaximum=1
     TargetMinimum=0
     bClamp=False
     Value=&RigVMFunction_MathQuaternionToEuler(RotationOrder=ZYX, Value=@n8.RelativeTransform.Rotation).Result.Y
    ).Result
   )
   > &RigUnit_ModifyTransforms(
    ItemToModify="((Item=(Type=Bone,Name=\"dyn_hairMID_01_Base_r\"),Transform=(Rotation=(X=0,Y=0,Z=0,W=1),Translation=(X=-1.5,Y=0,Z=0),Scale3D=(X=1,Y=1,Z=1))),(Item=(Type=Bone,Name=\"dyn_hairMid_02_r\"),Transform=(Rotation=(X=0,Y=-0.2247,Z=0,W=0.9744),Translation=(X=-1,Y=0,Z=0) ...(truncated)"
    WeightMinimum=0
    Weight=@Multiply_2.Result
   )
   n9 = &RigVMFunction_MathDoubleRemap(
    SourceMaximum=-55
    SourceMinimum=-10
    TargetMaximum=1
    TargetMinimum=0
    Value=&RigVMFunction_MathQuaternionToEuler(
     RotationOrder=XZY
     Value=&RigUnit_GetRelativeTransformForItem(Child=(Type=Bone,Name="head"), Parent=(Type=Bone,Name="root"), bChildInitial=False, bParentInitial=False).RelativeTransform.Rotation
    ).Result.Z
   )
   n10 = &RigVMFunction_MathQuaternionToEuler(
    RotationOrder=XZY
    Value=&RigUnit_GetRelativeTransformForItem(Child=(Type=Bone,Name="head"), Parent=(Type=Bone,Name="clavicle_r"), bChildInitial=False, bParentInitial=False).RelativeTransform.Rotation
   )
   n11 = /Script/RigVM.RigVMFunction_MathVectorMake(X=130.54, Y=161.807, Z=18.695) {is=struct RigVMUnitNode}
   n12 = &RigVMFunction_MathDoubleMul(
    B=@n9.Result
    A=&RigVMFunction_MathDoubleRemap(
     SourceMaximum=5
     SourceMinimum=40
     TargetMaximum=1
     TargetMinimum=0
     Value=&RigVMFunction_MathVectorDistance(A=@n10.Result, B=@n11.Result).Result
    ).Result
   )
   > &RigUnit_ModifyTransforms(
    ItemToModify="((Item=(Type=Bone,Name=\"dyn_hairMID_01_Base_r\"),Transform=(Rotation=(X=0.0036,Y=-0.0217,Z=-0.0051,W=0.9997),Translation=(X=-6,Y=0.4435,Z=1.919),Scale3D=(X=1,Y=1,Z=1))),(Item=(Type=Bone,Name=\"dyn_hairBack_01_c_BASE\"),Transform=(Rotation=(X=0,Y=0,Z=0,W=1),Tr ...(truncated)"
    WeightMinimum=0
    Weight=@n12.Result
   )
   RigVMFunction_MathVectorMake = &RigVMFunction_MathVectorMake(X=-141.5, Y=-138.5, Z=-9.3)
   RigVMFunction_MathVectorDistance = &RigVMFunction_MathVectorDistance(
    A=&RigVMFunction_MathQuaternionToEuler(
     RotationOrder=YZX
     Value=&RigUnit_GetRelativeTransformForItem(Child=(Type=Bone,Name="clavicle_r"), Parent=(Type=Bone,Name="head"), bChildInitial=False, bParentInitial=False).RelativeTransform.Rotation
    ).Result
    B=@RigVMFunction_MathVectorMake.Result
   )
   > &RigUnit_ModifyTransforms(
    ItemToModify=((Item=(Type=Bone,Name="dyn_hairMID_01_Base_r"),Transform=(Rotation=(X=0,Y=0,Z=0,W=1),Translation=(X=-1,Y=0,Z=0.0079),Scale3D=(X=1,Y=1,Z=1))))
    WeightMinimum=0
    Weight=&RigVMFunction_MathDoubleRemap(SourceMaximum=3, SourceMinimum=20, TargetMaximum=1, TargetMinimum=0, Value=@RigVMFunction_MathVectorDistance.Result).Result
   )
   > n13 = &RigUnit_SphericalPoseReader(
    ActiveRegionScaleFactors=(PositiveWidth=1,NegativeWidth=1,PositiveHeight=1,NegativeHeight=1)
    ActiveRegionSize=0.03
    Debug=(bDrawDebug=false,bDraw2D=False,bDrawLocalAxes=False,DebugScale=25,DebugSegments=20,DebugThickness=0.25)
    DriverAxis=(X=0,Y=1,Z=0)
    DriverItem=(Type=Bone,Name="head")
    FalloffRegionScaleFactors=(PositiveWidth=1,NegativeWidth=1,PositiveHeight=1,NegativeHeight=1)
    FalloffSize=0.2
    FlipHeightScaling=False
    FlipWidthScaling=False
    OptionalParentItem=(Type=Bone,Name="clavicle_r")
    RotationOffset=(X=0,Y=-87,Z=-5)
   )
   > &RigUnit_ModifyTransforms(
    ItemToModify="((Item=(Type=Bone,Name=\"dyn_hairMID_01_Base_r\"),Transform=(Rotation=(X=0,Y=0,Z=0,W=1),Translation=(X=-8.0635,Y=0,Z=0),Scale3D=(X=1,Y=1,Z=1))),(Item=(Type=Bone,Name=\"dyn_hairBack_01_c_BASE\"),Transform=(Rotation=(X=0.0151,Y=0.0858,Z=-0.173,W=0.9811),Translat ...(truncated)"
    WeightMinimum=0
    Weight=@n13.OutputParam
   )
   n14 = &RigVMFunction_MathVectorDistance(
    A=&RigVMFunction_MathQuaternionToEuler(
     RotationOrder=YZX
     Value=&RigUnit_GetRelativeTransformForItem(Child=(Type=Bone,Name="clavicle_r"), Parent=(Type=Bone,Name="head"), bChildInitial=False, bParentInitial=False).RelativeTransform.Rotation
    ).Result
    B=&RigVMFunction_MathVectorMake(X=-127.1, Y=-156, Z=-61.94).Result
   )
   > &RigUnit_ModifyTransforms(
    ItemToModify="((Item=(Type=Bone,Name=\"dyn_hairMID_01_Base_r\"),Transform=(Rotation=(X=0.0027,Y=0,Z=0,W=1),Translation=(X=-3.8962,Y=2.571,Z=1.3779),Scale3D=(X=1,Y=1,Z=1))),(Item=(Type=Bone,Name=\"dyn_hairBack_01_c_BASE\"),Transform=(Rotation=(X=0,Y=0,Z=-0.0754,W=0.9972),Tra ...(truncated)"
    WeightMinimum=0
    Weight=&RigVMFunction_KalmanFloat(
     BufferSize=16
     Value=&RigVMFunction_MathDoubleRemap(SourceMaximum=3, SourceMinimum=20, TargetMaximum=1, TargetMinimum=0, Value=@n14.Result).Result
    ) {template="KalmanFilter::Execute(in Value,in BufferSize,out Result)"}.Result
   )
   RigVMDispatchNode(Enabled=true, ExecuteContext=(), ScreenColor=(R=1,G=1,B=1,A=1), ScreenDuration=0.05, Value=0) {template="DISPATCH_RigVMDispatch_Print(in Prefix,in Value,in Enabled,in ScreenDuration,in ScreenColor)"}
   RigVMDispatchNode(Enabled=true, ExecuteContext=(), ScreenColor=(R=1,G=1,B=1,A=1), ScreenDuration=0.05, Value=@Multiply_2.Result) {template="DISPATCH_RigVMDispatch_Print(in Prefix,in Value,in Enabled,in ScreenDuration,in ScreenColor)"}
   &RigVMFunction_KalmanFloat(BufferSize=16, Value=0) {template="KalmanFilter::Execute(in Value,in BufferSize,out Result)"}
   &RigVMFunction_KalmanFloat(BufferSize=4, Value=@n12.Result) {template="KalmanFilter::Execute(in Value,in BufferSize,out Result)"}
   n15 = RigVMDispatchNode(Values="((Rotation=(X=0.1271,Y=0.0248,Z=-0.0032,W=0.9916),Translation=(X=0,Y=0,Z=0),Scale3D=(X=1,Y=1,Z=1)),((Rotation=(X=0,Y=0,Z=0,W=1),Translation=(X=0,Y=0,Z=0),Scale3D=(X=1,Y=1,Z=1)),Rotation=(X=-0.1502,Y=0.1373,Z=0.0211,W=0.9789),Translation=(X=0,Y=0,Z=0),Scale ...(truncated)") {template="DISPATCH_RigVMDispatch_ArrayMake(in Values,out Array)"}
   F_Head Rotation_2_1 = RigVMFunctionReferenceNode()
   &RigUnit_ModifyTransforms(
    ExecutePin=()
    ItemToModify="((Item=(Type=Bone,Name=\"dyn_hairBack_01_Base_l\"),Transform=(Rotation=(X=0.1271,Y=0.0248,Z=-0.0032,W=0.9916),Translation=(X=0,Y=0,Z=0),Scale3D=(X=1,Y=1,Z=1))),(Item=(Type=Bone,Name=\"dyn_hairMid_01_Base_l\"),Transform=(Rotation=(X=-0.1502,Y=0.1373,Z=0.0211,W= ...(truncated)"
    WeightMaximum=0
    WeightMinimum=0
    ItemToModify.2.Transform=RigVMDispatchNode(Index=2, Array=@n15.Array) {template="DISPATCH_RigVMDispatch_ArrayGetAtIndex(in Array,in Index,out Element)"}.Element
    ItemToModify.1.Transform=RigVMDispatchNode(Index=1, Array=@n15.Array) {template="DISPATCH_RigVMDispatch_ArrayGetAtIndex(in Array,in Index,out Element)"}.Element
    ItemToModify.0.Transform=RigVMDispatchNode(Index=0, Array=@n15.Array) {template="DISPATCH_RigVMDispatch_ArrayGetAtIndex(in Array,in Index,out Element)"}.Element
    Weight=&RigVMFunction_MathDoubleRemap(SourceMaximum=-69, SourceMinimum=0, TargetMaximum=1, TargetMinimum=0, bClamp=False, Value=@"F_Head Rotation_2_1".Result.X).Result
   )
   &RigVMFunction_MathDoubleMul(
    B=@n9.Result
    A=&RigVMFunction_MathDoubleRemap(
     SourceMaximum=0.005
     SourceMinimum=0.2
     TargetMaximum=1
     TargetMinimum=0
     bClamp=False
     Value=/Script/RigVM.RigVMFunction_MathVectorAngle(B=(X=130.54,Y=161.807,Z=18.695), A=@n10.Result) {is=struct RigVMUnitNode, template="AngleBetween::Execute(in A,in B,out Result)"}.Result
    ).Result
   )
   n16 = &RigUnit_GetRelativeTransformForItem(Child=(Type=Bone,Name="clavicle_r"), Parent=(Type=Bone,Name="head"), bChildInitial=False, bParentInitial=False)
   &RigVMFunction_MathDoubleRemap(
    SourceMaximum=3
    SourceMinimum=200
    TargetMaximum=1
    TargetMinimum=0
    Value=&RigVMFunction_MathVectorDistance(
     A=&RigVMFunction_MathQuaternionToEuler(RotationOrder=YZX, Value=@n16.RelativeTransform.Rotation).Result
     B=&RigVMFunction_MathVectorMake(X=-82.23, Y=167.969, Z=-61.117).Result
    ).Result
   )
  ]}
]
hierarchy [
 root {loc=<0 0 0>, rot=<0 0 0 1>}
 pelvis {parent=root, loc=<0 -0.7071 93.1191>, rot=<0.0127 0.707 -0.0127 -0.707>}
 spine_01 {parent=pelvis, loc=<6.7494 -0.6388 0>, rot=<0 0 0.034 -0.9994>}
 spine_02 {parent=spine_01, loc=<7.2518 -0.035 0>, rot=<0 0 0.0238 -0.9997>}
 spine_03 {parent=spine_02, loc=<6.0104 0 0>, rot=<0 0 0.0149 -0.9999>}
 spine_04 {parent=spine_03, loc=<8.1024 0 0>, rot=<0 0 -0.0132 -0.9999>}
 spine_05 {parent=spine_04, loc=<10.0146 0 0>, rot=<0 0 -0.0015 -1>}
 neck_01 {parent=spine_05, loc=<11.3138 0.1529 0>, rot=<0 0 -0.2003 -0.9797>}
 neck_02 {parent=neck_01, loc=<4.5627 0 0>, rot=<0 0 0.0504 -0.9987>}
 head {parent=neck_02, loc=<5.3914 0 0>, rot=<0 0 0.115 -0.9934>}
 dyn_curl_R_0 {parent=head, loc=<12.1265 8.748 3.3017>, rot=<-0.2333 -0.9723 -0.0036 -0.0146>}
 dyn_curl_R_1 {parent=dyn_curl_R_0, loc=<1.9669 0 0>, rot=<-0.0201 -0.0993 0.2742 -0.9563>}
 dyn_curl_R_2 {parent=dyn_curl_R_1, loc=<3.1267 0 0>, rot=<0 0 0 1>}
 dyn_curl_L_0 {parent=head, loc=<7.838 9.0355 -5.2048>, rot=<0.0504 -0.9957 -0.0033 0.0773>}
 dyn_curl_L_1 {parent=dyn_curl_L_0, loc=<2.0223 0 0>, rot=<0 0.1253 -0.0921 -0.9878>}
 dyn_hairBack_01_c_BASE {parent=head, loc=<6.7451 -2.5519 0>, rot=<0.2129 -0.9771 0.0008 -0.0039>}
 dyn_hairBack_01_c {parent=dyn_hairBack_01_c_BASE, loc=<0 0 0>, rot=<0 0 0 1>}
 dyn_hairBack_02_c {parent=dyn_hairBack_01_c, loc=<9.4538 0 0>, rot=<-0.0022 0.0146 0.1239 -0.9922>}
 dyn_hairBack_01_Base_l {parent=head, loc=<7.5999 1.2435 -6.2194>, rot=<0.1007 -0.9941 -0.0159 -0.0361>}
 dyn_hairBack_01_l {parent=dyn_hairBack_01_Base_l, loc=<0 0 0>, rot=<-0.0151 -0.0013 0.0858 -0.9962>}
 dyn_hairBack_02_l {parent=dyn_hairBack_01_l, loc=<8.6182 0 0>, rot=<0.0129 0.0228 0.1815 -0.983>}
 dyn_hairBack_01_Base_r {parent=head, loc=<7.6569 1.3968 5.6017>, rot=<0.1024 -0.9929 0.0107 0.0587>}
 dyn_hairBack_01_r {parent=dyn_hairBack_01_Base_r, loc=<0 0 0>, rot=<0 0 0.0858 -0.9963>}
 dyn_hairBack_02_r {parent=dyn_hairBack_01_r, loc=<8.8354 0 0>, rot=<-0.021 0.0187 0.1749 -0.9842>}
 dyn_hairMID_01_Base_r {parent=head, loc=<6.4574 -2.9571 3.8232>, rot=<0.0876 -0.4005 0.8971 -0.1646>}
 dyn_hairMid_01_r {parent=dyn_hairMID_01_Base_r, loc=<0 0 0>, rot=<0 0 0 1>}
 dyn_hairMid_02_r {parent=dyn_hairMid_01_r, loc=<7.2058 -0.2829 0>, rot=<-0.0206 0.0191 0.1942 -0.9806>}
 dyn_hairMid_01_Base_l {parent=head, loc=<6.4574 -2.9571 -3.8232>, rot=<0.1646 -0.8971 0.4005 -0.0876>}
 dyn_hairMid_01_l {parent=dyn_hairMid_01_Base_l, loc=<0 0 0>, rot=<0.1266 -0.0526 -0.1219 -0.983>}
 dyn_hairMid_02_l {parent=dyn_hairMid_01_l, loc=<7.2062 0.283 -0.0001>, rot=<0.0206 0.0191 -0.1942 -0.9806>}
 clavicle_l {parent=spine_05, loc=<4.3575 -0.1466 -2.3223>, rot=<0.0465 -0.7289 -0.009 -0.6829>}
 upperarm_l {parent=clavicle_l, loc=<15.5111 0 0>, rot=<0.0226 -0.3234 -0.0088 -0.9459>}
 clavicle_r {parent=spine_05, loc=<4.3575 -0.1466 2.3223>, rot=<0.7289 0.0465 -0.6829 0.009>}
 upperarm_r {parent=clavicle_r, loc=<-15.5111 0 0>, rot=<0.0226 -0.3234 -0.0088 -0.9459>}
 lowerarm_r {parent=upperarm_r, loc=<-26.6514 0 0>, rot=<0.0391 -0.0099 -0.2453 -0.9686>}
 hand_r {parent=lowerarm_r, loc=<-22.3491 0 0>, rot=<-0.7081 -0.0507 0.0975 -0.6975>}
 hand_attach_r {parent=hand_r, loc=<-8.5001 2.9999 0>, rot=<0 0 -0.7071 -0.7071>}
 thumb_01_r {parent=hand_r, loc=<-1.3118 0.3191 2.459>, rot=<0.6546 -0.4042 0.0664 -0.6354>}
 thumb_02_r {parent=thumb_01_r, loc=<-5.9336 0.0448 -0.0401>, rot=<-0.0037 0.0213 0.298 -0.9543>}
 thumb_03_r {parent=thumb_02_r, loc=<-2.2223 -0.3682 0.0907>, rot=<0.0118 -0.013 0.0066 -0.9998>}
 index_metacarpal_r {parent=hand_r, loc=<-1.6535 -0.0183 2.96>, rot=<0.0158 -0.0124 -0.0173 -0.9996>}
 index_01_r {parent=index_metacarpal_r, loc=<-8.3008 -0.0357 0.0524>, rot=<-0.0044 0.0485 0.0924 -0.9945>}
 index_02_r {parent=index_01_r, loc=<-3.3587 0.3004 -0.0452>, rot=<-0.002 0.0027 0.3225 -0.9466>}
 index_03_r {parent=index_02_r, loc=<-2.5578 0.123 0.0252>, rot=<-0.002 0.0067 0.2558 -0.9667>}
 middle_metacarpal_r {parent=hand_r, loc=<-2.1856 0.1725 0.5984>, rot=<0.006 0.0168 -0.0329 -0.9993>}
 middle_01_r {parent=middle_metacarpal_r, loc=<-7.3275 0.2447 -0.0129>, rot=<-0.0208 0.0042 0.0858 -0.9961>}
 middle_02_r {parent=middle_01_r, loc=<-4.0123 -0.1401 -0.1175>, rot=<-0.0218 0.0258 0.3512 -0.9357>}
 middle_03_r {parent=middle_02_r, loc=<-2.6077 -0.0534 -0.1373>, rot=<0.0082 0.0188 0.1708 -0.9851>}
 ring_metacarpal_r {parent=hand_r, loc=<-2.2074 0.3625 -1.3638>, rot=<-0.0284 0.0722 -0.0458 -0.9959>}
 ring_01_r {parent=ring_metacarpal_r, loc=<-7.0295 0.6197 0.1446>, rot=<-0.0489 -0.0252 0.1892 -0.9804>}
 ring_02_r {parent=ring_01_r, loc=<-3.6174 -0.4647 0.1027>, rot=<0 0 0.2806 -0.9598>}
 ring_03_r {parent=ring_02_r, loc=<-2.4152 -0.114 -0.0199>, rot=<0 0 0.1189 -0.9929>}
 pinky_metacarpal_r {parent=hand_r, loc=<-1.7616 0.9753 -3.3814>, rot=<-0.0397 0.0814 -0.069 -0.9935>}
 pinky_01_r {parent=pinky_metacarpal_r, loc=<-7.1709 0.4831 0.1094>, rot=<-0.0654 -0.0035 0.2242 -0.9723>}
 pinky_02_r {parent=pinky_01_r, loc=<-2.8274 -0.0893 0.1043>, rot=<0.0037 0.0011 0.2854 -0.9584>}
 pinky_03_r {parent=pinky_02_r, loc=<-2.2046 -0.0307 0.0078>, rot=<-0.0002 0.0098 0.1445 -0.9895>}
 lowerarm_twist_02_r {parent=lowerarm_r, loc=<-9.6199 0 0>, rot=<0 0 0 -1>}
 lowerarm_twist_01_r {parent=lowerarm_r, loc=<-17.8549 0 0>, rot=<0 0 0 -1>}
 clavicle_r_null {parent=spine_05, type=Null, loc=<4.3575 -0.1466 2.3223>, rot=<0.7289 0.0465 -0.6829 0.009>}
 head_null {parent=spine_05, type=Null, loc=<20.6573 3.5489 0>, rot=<0 0 -0.036 -0.9994>}
]
