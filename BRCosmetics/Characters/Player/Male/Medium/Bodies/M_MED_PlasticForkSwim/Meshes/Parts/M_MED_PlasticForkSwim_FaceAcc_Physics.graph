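# Physics asset for the face accessory part of the male medium PlasticForkSwim character cosmetic, simulating dynamic hair and curls.
# Capsules on the head, neck, upper spine and clavicles form body volumes alongside sphere and capsule hair chains.
# Back and mid chains hinge from base spheres and curls from the head, with tight swing limits and mostly locked twist.
# Angular-free cross constraints tie both mid hair tips to the center back tip.
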PhysicsAsset /BRCosmetics/Characters/Player/Male/Medium/Bodies/M_MED_PlasticForkSwim/Meshes/Parts/M_MED_PlasticForkSwim_FaceAcc_Physics
bodies {
 clavicle_l [{
   shapeType=Capsule
   length=8.451
   radius=8.3078
   shapeName=clavicle_r_capsule
   center=<6.6436 0 0.6366>
   rotation=<rot -82.122 179.9996 0.0003>
  }]
 clavicle_r [{
   shapeType=Capsule
   length=8.451
   radius=8.3078
   shapeName=clavicle_r_capsule
   center=<-6.6436 0 -0.6366>
   rotation=<rot -82.122 179.9996 -179.9997>
  }]
 dyn_curl_L_0 [{shapeType=Sphere, radius=1.0035, shapeName=dyn_curl_R_0_sphere, center=<1.7599 0 0>}]
 dyn_curl_L_1 [{shapeType=Sphere, radius=1.0035, shapeName=dyn_curl_R_0_sphere, center=<1.8132 0 0>}]
 dyn_curl_R_0 [{shapeType=Sphere, radius=1.0035, shapeName=dyn_curl_R_0_sphere, center=<1.8813 0 0>}]
 dyn_curl_R_1 [{shapeType=Sphere, radius=0.9153, shapeName=dyn_curl_R_0_sphere, center=<2.4735 0 0>}]
 dyn_curl_R_2 [{shapeType=Sphere, radius=0.9153, shapeName=dyn_curl_R_0_sphere, center=<2.4742 0 0>}]
 dyn_hairBack_01_Base_l [{shapeType=Sphere, radius=2.0459, shapeName=dyn_hairMid_01_Base_l_sphere}]
 dyn_hairBack_01_Base_r [{shapeType=Sphere, radius=2.0459, shapeName=dyn_hairMid_01_Base_l_sphere}]
 dyn_hairBack_01_c [{
   shapeType=Capsule
   length=4.6003
   radius=2.4529
   shapeName=dyn_hairBack_01_c_capsule
   center=<7.974 -1.2983 -0.0009>
   rotation=<rot -83.9533 -85.9862 90.3215>
  }]
 dyn_hairBack_01_c_BASE [{shapeType=Sphere, radius=2.0422, shapeName=dyn_hairBack_01_c_BASE_sphere}]
 dyn_hairBack_01_l [{
   shapeType=Capsule
   length=8.2479
   radius=1.6162
   shapeName=dyn_hairBack_01_l_capsule
   center=<8.1887 0 1.1787>
   rotation=<rot -90 0.0026 -0.0024>
  }]
 dyn_hairBack_01_r [{
   shapeType=Capsule
   length=6.5725
   radius=2.1391
   shapeName=dyn_hairBack_01_r_capsule
   center=<8.0401 0 0>
   rotation=<rot 90 179.7468 179.7467>
  }]
 dyn_hairBack_02_c [{
   shapeType=Capsule
   length=5.9483
   radius=2.9606
   shapeName=dyn_hairBack_02_c_capsule
   center=<3.138 0.2552 0>
   rotation=<rot 90 -179.9208 -179.921>
  }]
 dyn_hairBack_02_l [{
   shapeType=Capsule
   length=4.1303
   radius=2.8093
   shapeName=dyn_hairBack_02_l_capsule
   center=<2.996 3.3437 0.8498>
   rotation=<rot 0 33.1779 89.9999>
  }]
 dyn_hairBack_02_r [{
   shapeType=Capsule
   length=2.0345
   radius=2.1222
   shapeName=dyn_hairBack_02_r_capsule
   center=<1.4517 5.6955 0>
   rotation=<rot 0.0001 -119.9991 -90>
  }]
 dyn_hairMID_01_Base_r [{shapeType=Sphere, radius=2.0459, shapeName=dyn_hairMid_01_Base_l_sphere}]
 dyn_hairMid_01_Base_l [{shapeType=Sphere, radius=2.0459, shapeName=dyn_hairMid_01_Base_l_sphere}]
 dyn_hairMid_01_l [{shapeType=Capsule, length=2.5518, radius=2.4485, shapeName=dyn_hairMid_01_l_capsule, center=<5.6892 0 0>, rotation=<rot -90 180 90.0003>}]
 dyn_hairMid_01_r [{
   shapeType=Capsule
   length=3.9168
   radius=2.0304
   shapeName=dyn_hairMid_01_r_capsule
   center=<4.2959 -0.1633 0>
   rotation=<rot -0.0002 -0.0001 -90>
  }]
 dyn_hairMid_02_l [{
   shapeType=Capsule
   length=5.3453
   radius=2.2763
   shapeName=dyn_hairMid_02_l_capsule
   center=<5.6541 -0.0594 1.0621>
   rotation=<rot -0.0004 2.1229 89.9998>
  }]
 dyn_hairMid_02_r [{
   shapeType=Capsule
   length=4.1025
   radius=2.3659
   shapeName=dyn_hairBack_02_l_capsule
   center=<5.615 0.4965 0.1776>
   rotation=<rot -0.0002 6.0031 -89.9999>
  }]
 head [
  {shapeType=Capsule, length=6.5999, radius=6.2529, shapeName=head_capsule, center=<3.4929 3.5073 0>, rotation=<rot -90 0.0003 -0.0004>}
  {shapeType=Capsule, length=3.1156, radius=5.8885, shapeName=head_capsule, center=<6.8802 1.9038 0>, rotation=<rot 0 -10.0004 -90>}
 ]
 neck_01 [{shapeType=Capsule, length=5.9894, radius=4.714, shapeName=neck_01_capsule, center=<4.2843 0 0>, rotation=<rot -90 180 179.9999>}]
 spine_05 [{shapeType=Capsule, length=13.6101, radius=10.3875, shapeName=spine_05_capsule, center=<3.8892 -0.3873 0>}]
}
constraints [
 {bone1Name=dyn_hairBack_02_c, bone2Name=dyn_hairBack_01_c, swing1LimitDegrees=30, swing2LimitDegrees=30, twistLimitDegrees=30}
 {
  bone1Name=dyn_hairMid_02_l
  bone2Name=dyn_hairMid_01_l
  swing1LimitDegrees=30
  swing2LimitDegrees=30
  twistLimitDegrees=30
  twistMotion=Locked
 }
 {
  bone1Name=dyn_hairBack_01_c
  bone2Name=dyn_hairBack_01_c_BASE
  swing1LimitDegrees=20
  swing2LimitDegrees=5
  twistLimitDegrees=30
  twistMotion=Locked
 }
 {
  bone1Name=dyn_hairMid_01_l
  bone2Name=dyn_hairMid_01_Base_l
  swing1LimitDegrees=5
  swing1Motion=Locked
  swing2LimitDegrees=20
  twistLimitDegrees=30
  twistMotion=Locked
 }
 {
  bone1Name=dyn_hairBack_01_l
  bone2Name=dyn_hairBack_01_Base_l
  swing1LimitDegrees=10
  swing2LimitDegrees=10
  twistLimitDegrees=30
  twistMotion=Locked
 }
 {
  bone1Name=dyn_hairBack_02_l
  bone2Name=dyn_hairBack_01_l
  swing1LimitDegrees=30
  swing2LimitDegrees=30
  twistLimitDegrees=30
  twistMotion=Locked
 }
 {
  bone1Name=dyn_hairMid_01_r
  bone2Name=dyn_hairMID_01_Base_r
  swing1LimitDegrees=5
  swing1Motion=Locked
  swing2LimitDegrees=20
  twistLimitDegrees=30
  twistMotion=Locked
 }
 {bone1Name=dyn_hairMid_02_r, bone2Name=dyn_hairMid_01_r, swing1LimitDegrees=30, swing2LimitDegrees=30, twistLimitDegrees=30}
 {
  bone1Name=dyn_hairBack_01_r
  bone2Name=dyn_hairBack_01_Base_r
  swing1LimitDegrees=15
  swing2LimitDegrees=15
  twistLimitDegrees=30
  twistMotion=Locked
 }
 {
  bone1Name=dyn_hairBack_02_r
  bone2Name=dyn_hairBack_01_r
  swing1LimitDegrees=30
  swing2LimitDegrees=30
  twistLimitDegrees=30
  twistMotion=Locked
 }
 {bone1Name=dyn_curl_R_2, bone2Name=dyn_curl_R_1, swing1LimitDegrees=15, swing2LimitDegrees=15, twistLimitDegrees=45, twistMotion=Locked}
 {bone1Name=dyn_curl_R_1, bone2Name=dyn_curl_R_0, swing1LimitDegrees=15, swing2LimitDegrees=15, twistLimitDegrees=45, twistMotion=Locked}
 {bone1Name=dyn_curl_R_0, bone2Name=head, swing1LimitDegrees=15, swing2LimitDegrees=15, twistLimitDegrees=45, twistMotion=Locked}
 {bone1Name=dyn_curl_L_0, bone2Name=head, swing1LimitDegrees=15, swing2LimitDegrees=15, twistLimitDegrees=45, twistMotion=Locked}
 {bone1Name=dyn_curl_L_1, bone2Name=dyn_curl_L_0, swing1LimitDegrees=15, swing2LimitDegrees=15, twistLimitDegrees=45, twistMotion=Locked}
 {
  bone1Name=clavicle_l
  bone2Name=spine_05
  swing1LimitDegrees=45
  swing1Motion=Free
  swing2LimitDegrees=45
  swing2Motion=Free
  twistLimitDegrees=45
  twistMotion=Free
 }
 {
  bone1Name=dyn_hairMid_02_l
  bone2Name=dyn_hairBack_02_c
  swing1LimitDegrees=45
  swing1Motion=Free
  swing2LimitDegrees=45
  swing2Motion=Free
  twistLimitDegrees=45
  twistMotion=Free
 }
 {
  bone1Name=dyn_hairBack_02_c
  bone2Name=dyn_hairMid_02_r
  swing1LimitDegrees=45
  swing1Motion=Free
  swing2LimitDegrees=45
  swing2Motion=Free
  twistLimitDegrees=45
  twistMotion=Free
 }
]
